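# Post-process material for the Chapter 7 StW storm gameplay cue, judging by its name and folder.
# Its small graph samples scene depth at the viewport UV and divides it by a Distance scalar parameter, giving a normalized depth ratio.
# That ratio connects to no material output in the graph, so the material appears unfinished or stripped down.
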Material /Game/GameplayCueNotifies/StWStorm/M_Ch7_Storm_PPMat
graph {
 properties {bAutomaticallySetUsageInEditor=false, materialDomain=MD_PostProcess, usedShadingModels=Default Lit}
 nodes [
  n1 = ScreenPosition()
  Divide(A=ComponentMask(None=SceneDepth(UVs=@n1.ViewportUV).R) {r=true}, B=ScalarParameter() {defaultValue=20000, parameterName=Distance})
 ]
}
