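# Translucent unlit StW storm/heavy-rain overlay that attaches its mesh to the camera outside the safe zone and collapses it inside.
# Two phase-offset, rotating, zooming smoke layers blend with a radially mapped scrolling layer, normal-map distorted and faded by depth and distance.
# Screen derivatives drive tint at top quality.
# Opacity fades in near the zone edge, with thickness from a custom ray-march against the zone cylinder or constants on cheaper paths.
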Material /Game/GameplayCueNotifies/StWStorm/M_HeavyRainPost1
graph {
 properties {
  bDisableDepthTest=true
  bUseTranslucencyVertexFog=false
  bUsedWithNanite=true
  blendMode=BLEND_Translucent
  blendableLocation=BL_SceneColorAfterDOF
  pixelDepthOffsetMode=PDOM_Legacy
  shadingModel=MSM_Unlit
  usedShadingModels=Unlit
 }
 nodes [
  Constant()
  Constant()
  ScalarParameter() {defaultValue=1, parameterName=TestBrightness}
  n1 = Constant()
  n2 = SceneDepth(UVs=ScreenPosition().ViewportUV)
  n3 = Saturate(None=Divide(A=ComponentMask(None=@n2.R) {r=true}, B=ScalarParameter() {defaultValue=50}))
  n4 = ScalarParameter() {defaultValue=0.15, parameterName=Param_4}
  n5 = ScalarParameter() {defaultValue=1, parameterName=Param_5}
  n6 = CameraVectorWS()
  n7 = ComponentMask(None=@n6) {g=true, r=true}
  n8 = VectorParameter() {parameterName=UVS3_D, defaultValue=<rgba 7.0064 6.4868 0.05 0.05>}
  n9 = ComponentMask(None=@n8.RGB) {g=true, r=true}
  n10 = TextureSample(UVs=Panner(Coordinate=Divide(A=@n7, B=@n9), Speed=AppendVector(A=@n8.B, B=@n8.B))) {samplerType=SAMPLERTYPE_Normal, texture=/Game/Packages/Fortress_Sky/Materials/2Normal}
  n11 = VectorParameter() {defaultValue=<rgba 1 1 0.25 0>}
  n12 = Multiply(
   A=MaterialFunctionCall() {materialFunction=&WorldPositionBehindTranslucency}.Result
   B=@n11.RGB
  )
  n13 = Multiply(A=CameraPositionWS(), B=@n11.RGB)
  n14 = SphereMask(A=@n12, B=@n13, Radius=ScalarParameter() {defaultValue=26729.6699, parameterName=DistortDistnaceBasedFade}) {hardnessPercent=0}
  n15 = ComponentMask(None=Multiply(A=Multiply(A=@n10.RGB, B=@n14), B=ScalarParameter() {defaultValue=0.3001, parameterName=NormalDistort})) {g=true, r=true}
  n16 = ComponentMask(None=CameraVectorWS()) {b=true}
  n17 = ScalarParameter() {defaultValue=0.1143, parameterName=AA}
  n18 = Add(
   A=Time()
   B=Multiply(A=MaterialFunctionCall() {materialFunction=&DitherTemporalAA}.Result, B=@n17) {constB=0.2}
  )
  n19 = MaterialFunctionCall() {materialFunction=&DitherTemporalAA}
  n20 = Add(A=Time(), B=Add(A=@n17, B=Multiply(A=@n19.Result, B=@n17) {constB=0.2}) {constA=0.2})
  n21 = Time()
  n22 = QualitySwitch(Default=ShadingPathSwitch(Default=LinearInterpolate(A=@n18, B=@n20), Mobile=@n21) {desc=Migrated from FeatureLevelSwitch}, Low=@n21)
  n23 = Multiply(A=Multiply(A=@n22, B=Sign(None=@n16))) {constB=-0.33}
  n24 = Frac(None=@n23)
  n25 = LinearInterpolate(Alpha=@n24) {constA=3, constB=0.5}
  n26 = Divide(A=@n7, B=ComponentMask(None=VectorParameter() {parameterName=UVS2, defaultValue=<rgba 1.5 1.5 0 0.05>}.RGB) {g=true, r=true})
  n27 = Rotator(Coordinate=Add(A=@n15, B=Multiply(A=@n25, B=@n26))) {centerX=0, centerY=0, speed=0.125}
  n28 = Abs(None=LinearInterpolate(Alpha=@n24) {constA=-1})
  n29 = Frac(None=Add(A=@n23) {constB=0.5})
  n30 = LinearInterpolate(Alpha=@n29) {constA=3, constB=0.5}
  n31 = Rotator(Coordinate=Add(A=@n15, B=Multiply(A=@n30, B=@n26))) {centerX=0, centerY=0, speed=0.1}
  n32 = Abs(None=LinearInterpolate(Alpha=@n29) {constA=-1})
  n33 = Add(
   A=Multiply(A=TextureSample(UVs=@n27) {texture=&T_Inky_Smoke_Tile}.G, B=OneMinus(None=@n28))
   B=Multiply(A=TextureSample(UVs=@n31) {texture=&T_Inky_Smoke_Tile}.G, B=OneMinus(None=@n32))
  )
  n34 = VectorParameter() {parameterName=UVS1, defaultValue=<rgba 5 0.05 0.05 0.1>}
  n35 = ComponentMask(None=@n6) {b=true}
  n36 = Add(
   A=Multiply(
    A=MaterialFunctionCall(Vector or UVs=@n7) {materialFunction=&VectorToRadialValue}.Vector Converted to Angle
    B=@n34.R
   )
   B=Multiply(A=ScalarParameter() {defaultValue=0.0381, parameterName=Distort}, B=@n35)
  )
  n37 = Multiply(A=Tangent!(None=@n35) {period=4}) {constB=-1}
  n38 = Multiply(A=@n34.G, B=@n37)
  n39 = Add(A=Panner(Coordinate=AppendVector(A=@n36, B=@n38), Time=@n22, Speed=AppendVector(A=@n34.B, B=@n34.B)), B=@n15)
  n40 = SphereMask(A=@n37, B=Constant(), Radius=ScalarParameter() {defaultValue=8.1772, parameterName=HeightFade}) {hardnessPercent=0}
  n41 = Multiply(A=@n40, B=@n40)
  n42 = ScalarParameter() {defaultValue=1, parameterName=Param_2}
  n43 = Saturate(None=SphereMask(A=@n12, B=@n13, Radius=ScalarParameter() {defaultValue=26914.9785, parameterName=DistnaceBasedFade}) {hardnessPercent=0})
  n44 = AppendVector(A=@n43, B=Constant() {r=0.8809})
  n45 = Multiply(A=@n43, B=ScalarParameter() {defaultValue=2, parameterName=DistanceFallOffHardness}) {constB=2}
  n46 = Multiply(
   A=LinearInterpolate(
    A=LinearInterpolate(A=@n4, B=@n5, Alpha=@n33) {constA=0.25, constB=1.5}
    B=TextureSample(UVs=@n39) {texture=&T_Inky_Smoke_Tile}.G
    Alpha=Multiply(A=@n41, B=@n42)
   ) {constA=1}
   B=Multiply(
    A=TextureSample(UVs=@n44) {samplerType=SAMPLERTYPE_LinearColor, texture=&AthenaCurveAtlas01}.R
    B=Saturate(None=@n45)
   )
  )
  n47 = Add(
   A=Multiply(A=@n3, B=ScalarParameter() {defaultValue=0.5, parameterName=ForIntensity})
   B=Multiply(A=ScalarParameter() {defaultValue=0.5, parameterName=TextureIntensity}, B=Multiply(A=@n3, B=@n46))
  )
  n48 = ScalarParameter() {defaultValue=15, parameterName=Param_3}
  n49 = Clamp(None=Power(Base=Multiply(A=@n47, B=@n48), Exp=ScalarParameter() {defaultValue=0.75, parameterName=FinalGama})) {maxDefault=0.9}
  n50 = Add(A=DDX(Value=@n49), B=DDY(Value=@n49))
  n51 = ScalarParameter() {defaultValue=0.4667, parameterName=DDXY_Exp}
  n52 = Abs(None=Multiply(A=Sign(None=@n50), B=Power(Base=@n50, Exp=@n51)))
  n53 = Constant() {r=0.5}
  n54 = LinearInterpolate(
   A=VectorParameter() {parameterName=Param_1, defaultValue=<rgba 0.2882 0.0466 0.5 0>}.RGB
   B=VectorParameter() {parameterName=Param_2, defaultValue=<rgba 0 0 -0.1 0>}.RGB
   Alpha=ShadingPathSwitch(Default=QualitySwitch(Default=@n52, Low=@n53, High=@n53, Medium=@n53), Mobile=Constant() {r=0.5}) {desc=Migrated from FeatureLevelSwitch}
  )
  StaticSwitchParameter(
   True=ShadingPathSwitch(Default=@n54, Mobile=@n1) {desc=Migrated from FeatureLevelSwitch}
   False=EyeAdaptationInverse(LightValueInput=@n54, AlphaInput=ScalarParameter() {defaultValue=0.5, parameterName=EyeAdaptionAmount}).EyeAdaptationInverse
  ) {parameterName=UseMobileVisuals}
  n55 = Multiply(A=ComponentMask(None=@n49) {r=true}) {constB=0.8}
  n56 = ScalarParameter() {defaultValue=1, parameterName=BP_Blend}
  n57 = ScalarParameter() {defaultValue=0.75, parameterName=LowQualStormThickness}
  MaterialExpressionScalarParameter_28 = ScalarParameter() {defaultValue=8, parameterName=MaxSteps}
  MaterialExpressionMultiply_51 = Multiply(A=CameraVectorWS()) {constB=-1}
  n58 = CollectionParameter() {collection=&FortniteMaterialParameters, parameterName=AthenaSafeZoneCenter}
  n59 = CollectionParameter() {collection=&FortniteMaterialParameters, parameterName=AthenaSafeZoneRadius}
  MaterialExpressionAppendVector_7 = AppendVector(A=ComponentMask(None=@n58) {g=true, r=true}, B=@n59)
  MaterialExpressionCameraPositionWS_2 = CameraPositionWS()
  MaterialExpressionScalarParameter_33 = ScalarParameter() {defaultValue=512, parameterName=StepUnit}
  n60 = ScreenPosition()
  MaterialExpressionDivide_3 = Divide(
   A=ComponentMask(None=SceneDepth(UVs=@n60.ViewportUV).R) {r=true}
   B=DotProduct(
    A=@MaterialExpressionMultiply_51
    B=MaterialFunctionCall() {materialFunction=&CameraDirectionVector}.Result
   )
  )
  n61 = Custom!(
   MaxSteps=@MaterialExpressionScalarParameter_28
   CameraVector=@MaterialExpressionMultiply_51
   PosRadius=@MaterialExpressionAppendVector_7
   CameraPosition=@MaterialExpressionCameraPositionWS_2
   StepUnit=@MaterialExpressionScalarParameter_33
   SceneDepth=@MaterialExpressionDivide_3
  ) {
   code="float3 Col = 0;\r\nfloat3 cylinder = PosRadius;\r\n\r\nfor(int i=0; i<MaxSteps; i++)\r\n\t\r\n{\r\n\tfloat TestRange = min(StepUnit * i * i,SceneDepth);\r\n\tfloat3 TracePosition = CameraPosition + CameraVector * TestRange;\r\n\tfloat Distance = clamp((length(TracePosition.xy - cylinder.xy) - PosRadius.z)/(StepUnit*(i+1)),0,1) * (1/MaxSteps);\r\n//\tCol = lerp(Col,Distance,0.5);\r\n\tCol += Distance;\r\n}\r\n\r\nreturn Col;"
   inputs [
    {inputName=MaxSteps, input {expression=&self:MaterialExpressionScalarParameter_28}}
    {inputName=CameraVector, input {expression=&self:MaterialExpressionMultiply_51}}
    {inputName=PosRadius, input {expression=&self:MaterialExpressionAppendVector_7}}
    {inputName=CameraPosition, input {expression=&self:MaterialExpressionCameraPositionWS_2}}
    {inputName=StepUnit, input {expression=&self:MaterialExpressionScalarParameter_33}}
    {inputName=SceneDepth, input {expression=&self:MaterialExpressionDivide_3}}
   ]
  }
  n62 = QualitySwitch(Default=Min(A=@n57, B=ComponentMask(None=@n61) {r=true}), Low=@n57, High=@n57, Medium=@n57)
  n63 = ScalarParameter() {defaultValue=0.75, parameterName=Mobile Storm Thickness}
  n64 = CollectionParameter() {collection=&FortniteMaterialParameters, parameterName=AthenaSafeZoneCenter}
  n65 = Subtract(A=ComponentMask(None=CameraPositionWS()) {g=true, r=true}, B=ComponentMask(None=@n64) {g=true, r=true})
  n66 = CollectionParameter() {collection=&FortniteMaterialParameters, parameterName=AthenaSafeZoneRadius}
  n67 = ScalarParameter() {defaultValue=100, parameterName=Hide distance}
  n68 = Divide(
   A=Subtract(
    A=MaterialFunctionCall(Vector 2=@n65) {materialFunction=/Engine/Functions/Engine_MaterialFunctions02/Utility/VectorLength}.V2 Length
    B=Subtract(A=@n66, B=@n67)
   )
   B=@n67
  )
  n69 = Saturate(None=@n68)
  n70 = Multiply(A=Multiply(A=Multiply(A=@n55, B=@n56), B=ShadingPathSwitch(Default=@n62, Mobile=@n63) {desc=Migrated from FeatureLevelSwitch}), B=@n69)
  StaticSwitchParameter(True=ShadingPathSwitch(Default=@n70, Mobile=@n1) {desc=Migrated from FeatureLevelSwitch}, False=@n70) {parameterName=UseMobileVisuals}
  n71 = Multiply(A=WorldPosition().XYZ) {constB=-1}
  n72 = LinearInterpolate(
   A=@n71
   B=MaterialFunctionCall!(Object Basis 1, Object Basis 2, Object Basis 3, Camera Space Position Offset) {materialFunction=/Engine/Functions/Engine_MaterialFunctions02/WorldPositionOffset/AttachMeshToTheCamera}.World Position Offset
   Alpha=Ceil(None=@n69)
  )
  StaticSwitchParameter(True=@n72, False=ShadingPathSwitch(Default=@n72, Mobile=@n71) {desc=Migrated from FeatureLevelSwitch}) {parameterName=UseMobileVisuals}
  StaticSwitchParameter(True=ScalarParameter() {defaultValue=1, parameterName=Debug}, False=@n68) {parameterName=Debug}
 ]
}
